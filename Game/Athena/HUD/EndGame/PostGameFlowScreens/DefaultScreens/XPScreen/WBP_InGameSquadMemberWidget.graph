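# Post-game squad member slot showing portrait, name, voice status, party leader icon and Keep Playing Together vote.
# Opting in plays the Ready animation and sound; undecided and some opt-out statuses gray the portrait, reversing Ready if previously opted in.
# Fill slots, and non-opted-in members when matchmaking starts, become "New Player" placeholders.
# Non-local portraits are scaled by NonUserScaleValue, and extension widgets matching the configured slot type fill an avatar overlay.
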WidgetBlueprint /Game/Athena/HUD/EndGame/PostGameFlowScreens/DefaultScreens/XPScreen/WBP_InGameSquadMemberWidget
animations [
 Selected {
  binding_count=1
  length_seconds=0.279
  object=&self:Selected
  sequence {bindings [Image_Selected_Effect {
     id=#1
     possessed_class=Image
     tracks [
      {
       type=MovieSceneFloatTrack
       display_name=Render Opacity
       sections [{type=MovieSceneFloatSection, channels [None {type=MovieSceneScriptingFloatChannel, keys [[0,1],[5.5803,0]]}]}]
      }
      {
       type=MovieScene2DTransformTrack
       sections [{
         type=MovieScene2DTransformSection
         channels [
          Scale.X {type=MovieSceneScriptingFloatChannel, keys [[0,2],[5.5803,0.9451]]}
          Scale.Y {type=MovieSceneScriptingFloatChannel, keys [[0,2],[5.5803,0.9451]]}
         ]
        }]
      }
     ]
    }]}
 }
 Ready {
  binding_count=3
  length_seconds=0.3
  object=&self:Ready
  sequence {bindings [
    ReadyCheckBox {id=#2, possessed_class=Image, tracks []}
    SizeBoxReadyText {
     id=#3
     possessed_class=SizeBox
     tracks [
      {
       type=MovieScene2DTransformTrack
       sections [{
         type=MovieScene2DTransformSection
         channels [
          Translation.X {type=MovieSceneScriptingFloatChannel}
          Translation.Y {type=MovieSceneScriptingFloatChannel, keys [[0,137.765],[2.7063,1.4096],[4.8947,-12.0113],[6,0]]}
         ]
        }]
      }
      {
       type=MovieSceneFloatTrack
       display_name=Render Opacity
       sections [{type=MovieSceneFloatSection, channels [None {type=MovieSceneScriptingFloatChannel, keys [[0,0],[3.3857,1]]}]}]
      }
     ]
    }
    Overlay_1 {
     id=#4
     possessed_class=Overlay
     tracks [{
       type=MovieScene2DTransformTrack
       sections [{
         type=MovieScene2DTransformSection
         channels [
          Translation.X {type=MovieSceneScriptingFloatChannel}
          Translation.Y {type=MovieSceneScriptingFloatChannel}
          Angle {type=MovieSceneScriptingFloatChannel}
          Scale.X {type=MovieSceneScriptingFloatChannel, keys [[0,0],[4.055,1.2117],[4.2987,1.0488],[6,0.8438]]}
          Scale.Y {type=MovieSceneScriptingFloatChannel, keys [[0,0],[4.055,1.2117],[4.2987,1.0488],[6,0.8438]]}
          Shear.X {type=MovieSceneScriptingFloatChannel}
          Shear.Y {type=MovieSceneScriptingFloatChannel}
         ]
        }]
      }]
    }
   ]}
 }
 Reset {
  binding_count=3
  length_seconds=0.05
  object=&self:Reset
  sequence {bindings [
    ReadyCheckBox {id=#2, possessed_class=Image, tracks []}
    SizeBoxReadyText {
     id=#3
     possessed_class=SizeBox
     tracks [
      {
       type=MovieScene2DTransformTrack
       sections [{
         type=MovieScene2DTransformSection
         channels [
          Translation.X {type=MovieSceneScriptingFloatChannel}
          Translation.Y {type=MovieSceneScriptingFloatChannel, const [137.765,1]}
         ]
        }]
      }
      {
       type=MovieSceneFloatTrack
       display_name=Render Opacity
       sections [{type=MovieSceneFloatSection, channels [None {type=MovieSceneScriptingFloatChannel, const [0,1]}]}]
      }
     ]
    }
    Overlay_1 {
     id=#4
     possessed_class=Overlay
     tracks [{
       type=MovieScene2DTransformTrack
       sections [{
         type=MovieScene2DTransformSection
         channels [
          Translation.X {type=MovieSceneScriptingFloatChannel}
          Translation.Y {type=MovieSceneScriptingFloatChannel}
          Angle {type=MovieSceneScriptingFloatChannel}
          Scale.X {type=MovieSceneScriptingFloatChannel, const [0,1]}
          Scale.Y {type=MovieSceneScriptingFloatChannel, const [0,1]}
          Shear.X {type=MovieSceneScriptingFloatChannel}
          Shear.Y {type=MovieSceneScriptingFloatChannel}
         ]
        }]
      }]
    }
   ]}
 }
]
graphs [
 BP_HandleExtensionWidgetCreated {nodes [
   Handle Extension Widget Created!() -> (then, Widget:User Widget Object Reference, SlotType:EUIExtensionSlot Enum=Invalid) {is=title K2Node_FunctionEntry}
   > Branch(Condition="Equal (Enum)"(
     A:EUIExtensionSlot Enum=Get SlotType() -> (SlotType:EUIExtensionSlot Enum=Invalid)
     B:EUIExtensionSlot Enum=Get TeammateStatusPortraitOverlayIGM() -> (TeammateStatusPortraitOverlayIGM:EUIExtensionSlot Enum=Invalid)
    ))
   > n1 = AddChildToOverlay!(
    execute
    self:Overlay Object Reference=Get Overlay_AvatarExtension() -> (Overlay_AvatarExtension:Overlay Object Reference)
    Content:Widget Object Reference=Get Widget() -> (Widget:User Widget Object Reference)
   ) -> (then, ReturnValue:Overlay Slot Object Reference)
   > SetVerticalAlignment(self=@n1, InVerticalAlignment=VAlign_Fill)
   > SetHorizontalAlignment(self=@n1, InHorizontalAlignment=HAlign_Fill)
  ]}
 EventGraph {nodes [
   Event OnSetupAsFillSlot() {is=title K2Node_Event}
   > SetVisibility(InVisibility=HitTestInvisible)
   > PlayAnimation(InAnimation=Get Reset() -> (Reset:Widget Animation Object Reference))
   > SetActiveWidget(
    self=Get CommonVisibilitySwitcher_Portraits() -> (CommonVisibilitySwitcher_Portraits:Common Visibility Switcher Object Reference)
    Widget=Get Image_ToBeFilledPortrait() -> (Image_ToBeFilledPortrait:Fort Lazy Image Object Reference)
   )
   > "SetText (Text)"(self=Get Text_SquadMemberName() -> (Text_SquadMemberName:Text Object Reference), InText=Get Text_ToBeFilled() -> (Text_ToBeFilled:Text))
   > SetVisibility(
    self=[
     Get VoiceIcon_Status() -> (VoiceIcon_Status:Fort Voice Chat Status Icon Object Reference)
     Get Image_PartyMemberIndicator() -> (Image_PartyMemberIndicator:Image Object Reference)
    ]
    InVisibility=Collapsed
   )
   n1 = Event OnIsPartyLeaderChanged!() -> (OutputDelegate, then, bIsPartyLeader:bool) {is=title K2Node_Event}
   > SetVisibility(
    self=Get Overlay_PartyLeaderIcon() -> (Overlay_PartyLeaderIcon:Overlay Object Reference)
    InVisibility=Select(Option 0=Collapsed, Option 1=SelfHitTestInvisible, Index=@n1.bIsPartyLeader)
   )
   n2 = Event OnMatchmakingStarted!() -> (OutputDelegate, then, bWasOptedIn:bool) {is=title K2Node_Event}
   > Branch(Condition=NOT Boolean(A=@n2.bWasOptedIn))
   > SetActiveWidget(self=Get CommonVisibilitySwitcher_Portraits(), Widget=Get Image_ToBeFilledPortrait())
   > "SetText (Text)"(self=Get Text_SquadMemberName(), InText=Get Text_ToBeFilled())
   > SetVisibility(self=[Get VoiceIcon_Status(), Get Image_PartyMemberIndicator()], InVisibility=Collapsed)
   Event OnUserSet() {is=title K2Node_Event}
   > SetVisibility(InVisibility=HitTestInvisible)
   Event OnWidgetReset() {is=title K2Node_Event}
   > SetVisibility(InVisibility=Collapsed)
   > SetVisibility(self=Get Overlay_PartyLeaderIcon(), InVisibility=Collapsed)
   > SetActiveWidget(
    self=Get CommonVisibilitySwitcher_Portraits()
    Widget=Get Image_SquadMemberPortrait() -> (Image_SquadMemberPortrait:Fort Lazy Image Object Reference)
   )
   > SetVisibility(self=[Get VoiceIcon_Status(), Get Image_PartyMemberIndicator()], InVisibility=HitTestInvisible)
   Event OnTransitionedFromOptOut() {is=title K2Node_Event}
   > SetActiveWidget(self=Get CommonVisibilitySwitcher_Portraits(), Widget=Get Image_SquadMemberPortrait())
   > SetVisibility(self=[Get VoiceIcon_Status(), Get Image_PartyMemberIndicator()], InVisibility=HitTestInvisible)
   Event Construct()
   > PlayAnimation(InAnimation=Get Reset())
   > UpdatePortraitUserSpecifiedScale()
   Event PreConstruct()
   > UpdatePortraitUserSpecifiedScale()
   n3 = Event OnKeepPlayingTogetherVotingStatusChanged!() -> (OutputDelegate, then, NewStatus:EKeepPlayingTogetherVotingStatus Enum=Undecided, bWasOptedIn:bool) {is=title K2Node_Event}
   > n4 = Switch on EKeepPlayingTogetherVotingStatus!(execute, Selection:EKeepPlayingTogetherVotingStatus Enum=@n3.NewStatus) -> (
    Undecided:Exec
    OptedIn:Exec
    OptedOut_Manual:Exec
    OptedOut_PartySizeIncreased:Exec
    OptedOut_PartyMemberInLobby:Exec
    OptedOut_PartyMemberOptedOut:Exec
    OptedOut_PartyMemberLeft:Exec
    OptedOut_Forced:Exec
    OptedOut_TimedOut:Exec
    OptedOut_PlayerReturnedToLobby:Exec
    OptedOut_OnlyPlayerInSquad:Exec
    OptedOut_Disconnected:Exec
   ) {is=title K2Node_SwitchEnum}
   PlayAnimation(execute=@n4.OptedIn, InAnimation=Get Ready() -> (Ready:Widget Animation Object Reference))
   > PlaySound2D(Sound=/Game/Sounds/Fort_UI_Sounds/FrontEnd/SC_UI_MatchMaking_ReadyUp_Cue)
   > SetColorAndOpacity(
    self:Image Object Reference=Get Image_SquadMemberPortrait()
    InColorAndOpacity:Linear Color Structure=Get OptedInFullOpacity() -> (OptedInFullOpacity:Linear Color Structure)
   )
   UnreadyAndGrayOut!(
    execute=[@n4.OptedOut_Forced, @n4.OptedOut_Manual, @n4.OptedOut_TimedOut, @n4.Undecided]
    self
    WasOptedInPreviously:bool=@n3.bWasOptedIn
    IsMatchmakingActive:bool=false
   ) -> (then)
  ]}
 UnreadyAndGrayOut {nodes [
   n1 = UnreadyAndGrayOut!() -> (then, WasOptedInPreviously:bool=~, IsMatchmakingActive:bool=~) {is=title K2Node_FunctionEntry}
   > n2 = Branch(Condition=@n1)
   > n3 = PlayAnimationReverse(InAnimation=Get Ready(), PlaybackSpeed=2)
   SetColorAndOpacity(
    execute=[@n3.then, @n2.else]
    self=Get Image_SquadMemberPortrait()
    InColorAndOpacity=Get OptedOutFadedGray() -> (OptedOutFadedGray:Linear Color Structure)
   )
  ]}
 UpdatePortraitUserSpecifiedScale {nodes [
   UpdatePortraitUserSpecifiedScale() {is=title K2Node_FunctionEntry}
   > SetUserSpecifiedScale!(
    execute
    self:Scale Box Object Reference=Get ScaleBox_Portrait() -> (ScaleBox_Portrait:Scale Box Object Reference)
    InUserSpecifiedScale:float=SelectFloat(A=1, B=Get NonUserScaleValue() -> (NonUserScaleValue:double), bPickA=Get bIsLocalPlayerSlot() -> (bIsLocalPlayerSlot:bool))
   ) -> (then)
  ]}
]
variables [
 NonUserScaleValue {type=number, default=0.8}
 OptedInFullOpacity {type=LinearColor, default=<rgba 1 1 1 1>}
 OptedOutFadedGray {type=LinearColor, default=<rgba 0.05 0.05 0.05 0.75>}
 TeammateStatusPortraitOverlayIGM {type=EUIExtensionSlot, default=Invalid}
 Text_ToBeFilled {type=string, default=New Player}
]
widgets {
 info {inheritedWidgetCount=6, parentClass=/Script/FortniteUI.AthenaIGMSquadMemberWidget, rootWidgetClass=&VerticalBox, widgetCount=34}
 widgets [
  VerticalBox_1 {widgetClassPath=&VerticalBox, widget=VerticalBox_1}
  ScaleBox_ItemSize {
   widgetClassPath=&ScaleBox
   bIsVariable=true
   parent=VerticalBox_1
   widget=ScaleBox_ItemSize
   properties {stretch=UserSpecified}
   slot {class=VerticalBoxSlot}
  }
  VerticalBox_3 {widgetClassPath=&VerticalBox, parent=ScaleBox_ItemSize, widget=VerticalBox_3, slot {class=ScaleBoxSlot}}
  ScaleBox_Portrait {
   widgetClassPath=&ScaleBox
   bIsVariable=true
   parent=VerticalBox_3
   widget=ScaleBox_Portrait
   properties {stretch=UserSpecified}
   slot {class=VerticalBoxSlot, properties {verticalAlignment=VAlign_Bottom}}
  }
  Overlay_0 {widgetClassPath=&Overlay, parent=ScaleBox_Portrait, widget=Overlay_0, slot {class=ScaleBoxSlot}}
  SizeBox_2 {
   widgetClassPath=&SizeBox
   parent=Overlay_0
   widget=SizeBox_2
   properties {bOverride_HeightOverride=true, bOverride_WidthOverride=true, heightOverride=220, widthOverride=220}
   slot {class=OverlaySlot, properties {horizontalAlignment=HAlign_Center, verticalAlignment=VAlign_Center}}
  }
  Image_Selected_Effect {
   widgetClassPath=&Image
   bIsVariable=true
   parent=SizeBox_2
   widget=Image_Selected_Effect
   properties {
    renderOpacity=0
    visibility=HitTestInvisible
    brush {
     bIsDynamicallyLoaded=false
     drawAs=Box
     imageType=NoImage
     mirroring=NoMirror
     resourceName=None
     resourceObject=/Game/UI/Frontend/Showdown/Materials/M_UI_SkewBoxBorder_TournamentPanelBorder
     tiling=NoTile
     imageSize=<128 128>
     margin=<ltrb 0.2 0.3 0.2 0.3>
     outlineSettings {
      bUseBrushTransparency=false
      roundingType=HalfHeightRadius
      width=0
      color {colorUseRule=UseColor_Specified, specifiedColor=<rgba 0 0 0 0>}
      cornerRadii=<0 0 0 0>
     }
     tintColor {colorUseRule=UseColor_Specified, specifiedColor=<rgba 1 1 1 1>}
     uVRegion {bIsValid=false, max=<0 0>, min=<0 0>}
    }
   }
   slot {class=SizeBoxSlot}
  }
  Image_Selected {
   widgetClassPath=&Image
   bIsVariable=true
   parent=Overlay_0
   widget=Image_Selected
   properties {
    visibility=Collapsed
    brush {
     bIsDynamicallyLoaded=false
     drawAs=Image
     imageType=NoImage
     mirroring=NoMirror
     resourceName=None
     resourceObject=/Game/UI/Frontend/Showdown/Materials/M_UI_SkewBox_SelectedBG
     tiling=NoTile
     imageSize=<32 32>
     margin=<ltrb 0.2 0.2 0.2 0.2>
     outlineSettings {
      bUseBrushTransparency=false
      roundingType=HalfHeightRadius
      width=0
      color {colorUseRule=UseColor_Specified, specifiedColor=<rgba 0 0 0 0>}
      cornerRadii=<0 0 0 0>
     }
     tintColor {colorUseRule=UseColor_Specified, specifiedColor=<rgba 1 1 1 1>}
     uVRegion {bIsValid=false, max=<0 0>, min=<0 0>}
    }
    renderTransform {angle=0, scale=<1 1>, shear=<2 0>, translation=<0 0>}
   }
   slot {class=OverlaySlot, properties {horizontalAlignment=HAlign_Fill, verticalAlignment=VAlign_Fill, padding=<ltrb -12 -10 -16 -10>}}
  }
  SizeBox_0 {
   widgetClassPath=&SizeBox
   parent=Overlay_0
   widget=SizeBox_0
   properties {bOverride_HeightOverride=true, bOverride_WidthOverride=true, heightOverride=270, widthOverride=270}
   slot {class=OverlaySlot, properties {horizontalAlignment=HAlign_Fill, verticalAlignment=VAlign_Fill}}
  }
  Overlay_2 {widgetClassPath=&Overlay, parent=SizeBox_0, widget=Overlay_2, slot {class=SizeBoxSlot}}
  CommonVisibilitySwitcher_Portraits {
   widgetClassPath=&CommonVisibilitySwitcher
   bIsVariable=true
   parent=Overlay_2
   widget=CommonVisibilitySwitcher_Portraits
   slot {class=OverlaySlot, properties {horizontalAlignment=HAlign_Fill, verticalAlignment=VAlign_Fill}}
  }
  Image_SquadMemberPortrait {
   widgetClassPath=&FortLazyImage
   bInherited=true
   bIsVariable=true
   parent=CommonVisibilitySwitcher_Portraits
   widget=Image_SquadMemberPortrait
   properties {brush {
     bIsDynamicallyLoaded=false
     drawAs=Image
     imageType=NoImage
     mirroring=NoMirror
     resourceName=None
     resourceObject=/Game/UI/Frontend/HeroLoadout/Art/Materials/Hero+Class/M_UI_LoadoutSlot_Portrait_Inst
     tiling=NoTile
     imageSize=<220 220>
     margin=<ltrb 0 0 0 0>
     outlineSettings {
      bUseBrushTransparency=false
      roundingType=HalfHeightRadius
      width=0
      color {colorUseRule=UseColor_Specified, specifiedColor=<rgba 0 0 0 0>}
      cornerRadii=<0 0 0 0>
     }
     tintColor {colorUseRule=UseColor_Specified, specifiedColor=<rgba 1 1 1 1>}
     uVRegion {bIsValid=false, max=<0 0>, min=<0 0>}
    }}
   slot {class=CommonVisibilitySwitcherSlot}
  }
  AvatarIconWidget {
   widgetClassPath=/Script/FortniteUI.FortSocialAvatarIcon
   bInherited=true
   bIsVariable=true
   parent=CommonVisibilitySwitcher_Portraits
   widget=AvatarIconWidget
   properties {visibility=Collapsed}
   slot {class=CommonVisibilitySwitcherSlot}
  }
  Image_ToBeFilledPortrait {
   widgetClassPath=&FortLazyImage
   bIsVariable=true
   parent=CommonVisibilitySwitcher_Portraits
   widget=Image_ToBeFilledPortrait
   properties {brush {
     bIsDynamicallyLoaded=false
     drawAs=Image
     imageType=NoImage
     mirroring=NoMirror
     resourceName=None
     resourceObject=&Hero+Class/M_UI_LoadoutSlot_Portrait_Inst
     tiling=NoTile
     imageSize=<86 86>
     margin=<ltrb 0 0 0 0>
     outlineSettings {
      bUseBrushTransparency=false
      roundingType=HalfHeightRadius
      width=0
      color {colorUseRule=UseColor_Specified, specifiedColor=<rgba 0 0 0 0>}
      cornerRadii=<0 0 0 0>
     }
     tintColor {colorUseRule=UseColor_Specified, specifiedColor=<rgba 0 0 0 0.75>}
     uVRegion {bIsValid=false, max=<0 0>, min=<0 0>}
    }}
   slot {class=CommonVisibilitySwitcherSlot}
  }
  Overlay_AvatarExtension {
   widgetClassPath=&Overlay
   bIsVariable=true
   parent=Overlay_2
   widget=Overlay_AvatarExtension
   slot {class=OverlaySlot, properties {horizontalAlignment=HAlign_Fill, verticalAlignment=VAlign_Fill}}
  }
  Overlay_1 {
   widgetClassPath=&Overlay
   parent=Overlay_0
   widget=Overlay_1
   properties {renderTransform {angle=0, scale=<1 1>, shear=<0 0>, translation=<8 8>}, renderTransformPivot=<0.5 1>}
   slot {class=OverlaySlot, properties {horizontalAlignment=HAlign_Right, verticalAlignment=VAlign_Bottom}}
  }
  ReadyCheckBox_Shadow {
   widgetClassPath=&Image
   bIsVariable=true
   parent=Overlay_1
   widget=ReadyCheckBox_Shadow
   properties {
    brush {
     bIsDynamicallyLoaded=false
     drawAs=Image
     imageType=NoImage
     mirroring=NoMirror
     resourceName=None
     resourceObject=/Game/UI/Foundation/Textures/Icons/Checks/T-Icon-Check-128
     tiling=NoTile
     imageSize=<58 58>
     margin=<ltrb 0 0 0 0>
     outlineSettings {
      bUseBrushTransparency=false
      roundingType=HalfHeightRadius
      width=0
      color {colorUseRule=UseColor_Specified, specifiedColor=<rgba 0 0 0 0>}
      cornerRadii=<0 0 0 0>
     }
     tintColor {colorUseRule=UseColor_Specified, specifiedColor=<rgba 1 1 1 1>}
     uVRegion {bIsValid=false, max=<0 0>, min=<0 0>}
    }
    colorAndOpacity=<rgba 0.0194 0.0194 0.0666 0.8>
    renderTransform {angle=4, scale=<1 1>, shear=<0 0>, translation=<4 7>}
    renderTransformPivot=<0.3439 1>
   }
   slot {class=OverlaySlot}
  }
  ReadyCheckBox {
   widgetClassPath=&Image
   bIsVariable=true
   parent=Overlay_1
   widget=ReadyCheckBox
   properties {
    brush {
     bIsDynamicallyLoaded=false
     drawAs=Image
     imageType=NoImage
     mirroring=NoMirror
     resourceName=None
     resourceObject=&T-Icon-Check-128
     tiling=NoTile
     imageSize=<58 58>
     margin=<ltrb 0 0 0 0>
     outlineSettings {
      bUseBrushTransparency=false
      roundingType=HalfHeightRadius
      width=0
      color {colorUseRule=UseColor_Specified, specifiedColor=<rgba 0 0 0 0>}
      cornerRadii=<0 0 0 0>
     }
     tintColor {colorUseRule=UseColor_Specified, specifiedColor=<rgba 1 1 1 1>}
     uVRegion {bIsValid=false, max=<0 0>, min=<0 0>}
    }
    colorAndOpacity=<rgba 0 1 0.1133 1>
   }
   slot {class=OverlaySlot}
  }
  SizeBoxReadyText {
   widgetClassPath=&SizeBox
   parent=Overlay_0
   widget=SizeBoxReadyText
   properties {bOverride_WidthOverride=true, renderTransformPivot=<0.5 1>}
   slot {class=OverlaySlot, properties {horizontalAlignment=HAlign_Fill, verticalAlignment=VAlign_Bottom, padding=<ltrb 0 0 0 -10>}}
  }
  ScaleBox_0 {
   widgetClassPath=&ScaleBox
   parent=SizeBoxReadyText
   widget=ScaleBox_0
   properties {stretchDirection=DownOnly}
   slot {class=SizeBoxSlot, properties {verticalAlignment=VAlign_Bottom}}
  }
  HorizontalBox_1 {
   widgetClassPath=&HorizontalBox
   parent=ScaleBox_0
   widget=HorizontalBox_1
   slot {class=ScaleBoxSlot, properties {horizontalAlignment=HAlign_Left, verticalAlignment=VAlign_Bottom}}
  }
  Overlay_PartyLeaderIcon {
   widgetClassPath=&Overlay
   bIsVariable=true
   parent=HorizontalBox_1
   widget=Overlay_PartyLeaderIcon
   properties {visibility=Collapsed}
   slot {class=HorizontalBoxSlot, properties {horizontalAlignment=HAlign_Left, verticalAlignment=VAlign_Center, padding=<ltrb 0 -2 0 0>}}
  }
  PartyLeaderIcon {
   widgetClassPath=&Image
   parent=Overlay_PartyLeaderIcon
   widget=PartyLeaderIcon
   properties {
    visibility=SelfHitTestInvisible
    brush {
     bIsDynamicallyLoaded=false
     drawAs=Image
     imageType=NoImage
     mirroring=NoMirror
     resourceName=None
     resourceObject=/Game/UI/Icons/Icon-Nodes/Icon-SquadLeaderSlot-Generic-XL
     tiling=NoTile
     imageSize=<46 46>
     margin=<ltrb 0 0 0 0>
     outlineSettings {
      bUseBrushTransparency=false
      roundingType=HalfHeightRadius
      width=0
      color {colorUseRule=UseColor_Specified, specifiedColor=<rgba 0 0 0 0>}
      cornerRadii=<0 0 0 0>
     }
     tintColor {colorUseRule=UseColor_Specified, specifiedColor=<rgba 1 1 1 1>}
     uVRegion {bIsValid=false, max=<0 0>, min=<0 0>}
    }
    colorAndOpacity=<rgba 1 0.8688 0 1>
   }
   slot {class=OverlaySlot}
  }
  Text_Readiness {
   widgetClassPath=&CommonTextBlock
   bIsVariable=true
   parent=HorizontalBox_1
   widget=Text_Readiness
   properties {
    justification=Center
    text=Ready
    textTransformPolicy=ToUpper
    font {
     bForceMonospaced=false
     bMaterialIsStencil=false
     fontMaterial=null
     fontObject=&Burbank
     letterSpacing=0
     monospacedWidth=1
     size=30
     skewAmount=0
     typefaceFontName=Burbank-PlayerPower
     outlineSettings {
      bApplyOutlineToDropShadows=false
      bMiteredCorners=false
      bSeparateFillAlpha=false
      outlineMaterial=null
      outlineSize=0
      outlineColor=<rgba 0.0052 0.0052 0.0056 0.2>
     }
    }
    renderTransform {angle=0, scale=<1 1>, shear=<-8 0>, translation=<0 0>}
    renderTransformPivot=<0.5 1>
    shadowColorAndOpacity=<rgba 0 0 0 0.5>
    shadowOffset=<3 2>
   }
   slot {class=HorizontalBoxSlot}
  }
  SizeBox_1 {
   widgetClassPath=&SizeBox
   parent=VerticalBox_3
   widget=SizeBox_1
   properties {bOverride_WidthOverride=true}
   slot {class=VerticalBoxSlot, properties {padding=<ltrb 0 15 0 5>}}
  }
  ScaleBox {
   widgetClassPath=&ScaleBox
   parent=SizeBox_1
   widget=ScaleBox
   properties {stretch=None}
   slot {class=SizeBoxSlot, properties {horizontalAlignment=HAlign_Center}}
  }
  HorizontalBox_0 {widgetClassPath=&HorizontalBox, parent=ScaleBox, widget=HorizontalBox_0, slot {class=ScaleBoxSlot}}
  Image_PartyMemberIndicator {
   widgetClassPath=&Image
   bInherited=true
   bIsVariable=true
   parent=HorizontalBox_0
   widget=Image_PartyMemberIndicator
   properties {
    visibility=Collapsed
    brush {
     bIsDynamicallyLoaded=false
     drawAs=Image
     imageType=NoImage
     mirroring=NoMirror
     resourceName=None
     resourceObject=/Game/UI/Frontend/Parties/PartiesButtonStyles/T_UI_Icon_QuadMeeple
     tiling=NoTile
     imageSize=<32 32>
     margin=<ltrb 0 0 0 0>
     outlineSettings {
      bUseBrushTransparency=false
      roundingType=HalfHeightRadius
      width=0
      color {colorUseRule=UseColor_Specified, specifiedColor=<rgba 0 0 0 0>}
      cornerRadii=<0 0 0 0>
     }
     tintColor {colorUseRule=UseColor_Specified, specifiedColor=<rgba 1 1 1 1>}
     uVRegion {bIsValid=false, max=<0 0>, min=<0 0>}
    }
   }
   slot {class=HorizontalBoxSlot}
  }
  SizeBox_3 {
   widgetClassPath=&SizeBox
   parent=HorizontalBox_0
   widget=SizeBox_3
   properties {bOverride_MaxDesiredWidth=true, maxDesiredWidth=230}
   slot {class=HorizontalBoxSlot, properties {horizontalAlignment=HAlign_Center, verticalAlignment=VAlign_Bottom}}
  }
  ScaleBox_1 {widgetClassPath=&ScaleBox, parent=SizeBox_3, widget=ScaleBox_1, slot {class=SizeBoxSlot}}
  CommonVisibilitySwitcher_1 {widgetClassPath=&CommonVisibilitySwitcher, parent=ScaleBox_1, widget=CommonVisibilitySwitcher_1, slot {class=ScaleBoxSlot}}
  Text_SquadMemberName {
   widgetClassPath=&CommonTextBlock
   bInherited=true
   bIsVariable=true
   parent=CommonVisibilitySwitcher_1
   widget=Text_SquadMemberName
   properties {
    justification=Center
    mobileFontSizeMultiplier=1.2
    text=VeryVeryLongName
    textTransformPolicy=ToUpper
    font {
     bForceMonospaced=false
     bMaterialIsStencil=false
     fontMaterial=null
     fontObject=&Burbank
     letterSpacing=0
     monospacedWidth=1
     size=18
     skewAmount=0
     typefaceFontName=BurbankRegular-Black
     outlineSettings {
      bApplyOutlineToDropShadows=false
      bMiteredCorners=false
      bSeparateFillAlpha=false
      outlineMaterial=null
      outlineSize=1
      outlineColor=<rgba 0 0 0 0.75>
     }
    }
    renderTransform {angle=0, scale=<1 1>, shear=<-8 0>, translation=<0 0>}
    renderTransformPivot=<0.5 1>
    shadowOffset=<0 0>
   }
   slot {class=CommonVisibilitySwitcherSlot}
  }
  Text_SocialName {
   widgetClassPath=/Script/FortniteUI.FortSocialNameTextBlock
   bInherited=true
   bIsVariable=true
   parent=CommonVisibilitySwitcher_1
   widget=Text_SocialName
   properties {
    bOverrideDefaultStyle=true
    bTintInlineIcon=true
    text="Nickname <PC color=\"82BDBDFF\" size=\"12\"></><text color=\"82BDBDFF\" size=\"12\">Display Name</>"
    visibility=Collapsed
    defaultTextStyleOverride {
     overflowPolicy=Clip
     transformPolicy=None
     colorAndOpacity {colorUseRule=UseColor_Specified, specifiedColor=<rgba 0.521 1 0.8228 1>}
     font {
      bForceMonospaced=false
      bMaterialIsStencil=false
      fontMaterial=null
      fontObject=&BurbankSmall
      letterSpacing=0
      monospacedWidth=1
      size=16
      skewAmount=0
      typefaceFontName=Black
      outlineSettings {
       bApplyOutlineToDropShadows=false
       bMiteredCorners=false
       bSeparateFillAlpha=false
       outlineMaterial=null
       outlineSize=0
       outlineColor=<rgba 0 0 0 1>
      }
     }
     highlightColor {colorUseRule=UseColor_Specified, specifiedColor=<rgba 0 0 0 1>}
     highlightShape {
      bIsDynamicallyLoaded=false
      drawAs=Image
      imageType=NoImage
      mirroring=NoMirror
      resourceName=None
      resourceObject=null
      tiling=NoTile
      imageSize=<32 32>
      margin=<ltrb 0 0 0 0>
      outlineSettings {
       bUseBrushTransparency=false
       roundingType=HalfHeightRadius
       width=0
       color {colorUseRule=UseColor_Specified, specifiedColor=<rgba 0 0 0 0>}
       cornerRadii=<0 0 0 0>
      }
      tintColor {colorUseRule=UseColor_Specified, specifiedColor=<rgba 1 1 1 1>}
      uVRegion {bIsValid=false, max=<0 0>, min=<0 0>}
     }
     selectedBackgroundColor {colorUseRule=UseColor_Foreground, specifiedColor=<rgba 1 0 1 1>}
     shadowColorAndOpacity=<rgba 0 0 0 1>
     shadowOffset=<0 0>
     strikeBrush {
      bIsDynamicallyLoaded=false
      drawAs=Image
      imageType=NoImage
      mirroring=NoMirror
      resourceName=None
      resourceObject=null
      tiling=NoTile
      imageSize=<32 32>
      margin=<ltrb 0 0 0 0>
      outlineSettings {
       bUseBrushTransparency=false
       roundingType=HalfHeightRadius
       width=0
       color {colorUseRule=UseColor_Specified, specifiedColor=<rgba 0 0 0 0>}
       cornerRadii=<0 0 0 0>
      }
      tintColor {colorUseRule=UseColor_Specified, specifiedColor=<rgba 1 1 1 1>}
      uVRegion {bIsValid=false, max=<0 0>, min=<0 0>}
     }
     underlineBrush {
      bIsDynamicallyLoaded=false
      drawAs=Image
      imageType=NoImage
      mirroring=NoMirror
      resourceName=None
      resourceObject=null
      tiling=NoTile
      imageSize=<32 32>
      margin=<ltrb 0 0 0 0>
      outlineSettings {
       bUseBrushTransparency=false
       roundingType=HalfHeightRadius
       width=0
       color {colorUseRule=UseColor_Specified, specifiedColor=<rgba 0 0 0 0>}
       cornerRadii=<0 0 0 0>
      }
      tintColor {colorUseRule=UseColor_Specified, specifiedColor=<rgba 1 1 1 1>}
      uVRegion {bIsValid=false, max=<0 0>, min=<0 0>}
     }
    }
   }
   slot {class=CommonVisibilitySwitcherSlot}
  }
  VoiceIcon_Status {
   widgetClassPath=/Script/FortniteUI.FortVoiceChatStatusIconProxyAnim
   bInherited=true
   bIsVariable=true
   parent=HorizontalBox_0
   widget=VoiceIcon_Status
   properties {renderTransform {angle=0, scale=<1 1>, shear=<0 0>, translation=<0 3>}}
   slot {class=HorizontalBoxSlot, properties {horizontalAlignment=HAlign_Left, verticalAlignment=VAlign_Bottom, padding=<ltrb 3 0 -24 0>}}
  }
 ]
}
