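# Squad display widget for Keep Playing Together on the post-game XP screen, derived from AthenaIGMSquadDisplayWidget.
# An Overlay layers a row of squad members over a skewed, left-aligned message background.
# The inherited DynamicEntryBox, inside a ScaleBox, spawns WBP_InGameSquadMemberWidget entries, while the inherited local player slot starts collapsed.
# The event graph holds only unwired events, leaving behavior to the parent class.
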WidgetBlueprint /Game/Athena/HUD/EndGame/PostGameFlowScreens/DefaultScreens/XPScreen/WBP_KeepPlayingTogetherIGMSquadDisplay
graphs [EventGraph {nodes [
   Event PreConstruct()
   Event Construct()
   Event Tick()
  ]}]
variables []
widgets {
 info {inheritedWidgetCount=2, parentClass=/Script/FortniteUI.AthenaIGMSquadDisplayWidget, rootWidgetClass=&Overlay, widgetCount=6}
 widgets [
  Overlay_0 {widgetClassPath=&Overlay, widget=Overlay_0}
  Image_SkewBox_BG {
   widgetClassPath=&Image
   bIsVariable=true
   parent=Overlay_0
   widget=Image_SkewBox_BG
   properties {brush {
     bIsDynamicallyLoaded=false
     drawAs=Box
     imageType=NoImage
     mirroring=NoMirror
     resourceName=None
     resourceObject=/Game/Athena/HUD/Materials/M_UI_MessageBG_Layered_Inst_LeftAligned
     tiling=NoTile
     imageSize=<350 32>
     margin=<ltrb 0.4 0 0.4 0>
     outlineSettings {
      bUseBrushTransparency=false
      roundingType=HalfHeightRadius
      width=0
      color {colorUseRule=UseColor_Specified, specifiedColor=<rgba 0 0 0 0>}
      cornerRadii=<0 0 0 0>
     }
     tintColor {colorUseRule=UseColor_Specified, specifiedColor=<rgba 1 1 1 1>}
     uVRegion {bIsValid=false, max=<0 0>, min=<0 0>}
    }}
   slot {class=OverlaySlot, properties {horizontalAlignment=HAlign_Fill, verticalAlignment=VAlign_Fill, padding=<ltrb 0 35 0 0>}}
  }
  HorizontalBox_66 {
   widgetClassPath=&HorizontalBox
   parent=Overlay_0
   widget=HorizontalBox_66
   properties {renderTransform {angle=0, scale=<1 1>, shear=<0 0>, translation=<0 -11>}}
   slot {class=OverlaySlot, properties {padding=<ltrb 40 0 120 0>}}
  }
  LocalPlayerSquadMemberWidget {
   widgetClassPath=&/WBP_InGameSquadMemberWidget
   bInherited=true
   bIsVariable=true
   parent=HorizontalBox_66
   widget=LocalPlayerSquadMemberWidget
   properties {bIsLocalPlayerSlot=true, visibility=Collapsed}
   slot {class=HorizontalBoxSlot}
  }
  ScaleBox_Players {
   widgetClassPath=&ScaleBox
   parent=HorizontalBox_66
   widget=ScaleBox_Players
   properties {stretch=UserSpecified}
   slot {class=HorizontalBoxSlot, properties {verticalAlignment=VAlign_Bottom, padding=<ltrb 75 0 0 0>}}
  }
  EntryBox_SquadMemberWidgets {
   widgetClassPath=&DynamicEntryBox
   bInherited=true
   bIsVariable=true
   parent=ScaleBox_Players
   widget=EntryBox_SquadMemberWidgets
   properties {
    entryWidgetClass=&/WBP_InGameSquadMemberWidget.WBP_InGameSquadMemberWidget_C
    entrySpacing=<100 0>
   }
   slot {class=ScaleBoxSlot, properties {horizontalAlignment=HAlign_Left}}
  }
 ]
}
