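# UI material for a tournament small banner thumbnail on the post-game XP screen, with a loading placeholder.
# The texture is scaled to fill and darkened by a squared gradient toward the bottom.
# IsLoading blends the color to white and the alpha to a dim base plus a horizontal sweep that fades over a repeating time cycle.
# A rounded-corner mask with a CornerRadius parameter caps the alpha through a Min.
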Material /Game/Athena/HUD/EndGame/PostGameFlowScreens/DefaultScreens/XPScreen/Art/M_TournamentSmallBannerThumbnail
graph {
 properties {blendMode=BLEND_Translucent, materialDomain=MD_UI, pixelDepthOffsetMode=PDOM_Legacy, usedShadingModels=Default Lit}
 nodes [
  n1 = Constant() {r=0.25}
  n2 = Multiply(A=NamedRerouteUsage()) {constB=2}
  Min(
   A=MaterialFunctionCall(Corner Radius=ScalarParameter() {defaultValue=12, parameterName=CornerRadius}, Radius in Pixels=Constant() {r=1}) {materialFunction=&MF_UI_CoreUI_RoundedMask}.Shape Mask
   B=LinearInterpolate(B=Add(A=@n1, B=@n2), Alpha=ScalarParameter() {defaultValue=1, parameterName=IsLoading}) {constA=1, constB=0.4}
  )
  n3 = TextureObjectParameter() {parameterName=Texture}
  n4 = MaterialFunctionCall(TextureSize=TextureProperty(Tex=@n3)) {materialFunction=&MF_UI_UVScaling}
  n5 = MaterialFunctionCall() {materialFunction=&LinearGradient}
  n6 = Constant() {desc=Bottom Gradient Size Percent, r=0.55}
  n7 = OneMinus(None=@n6)
  n8 = Saturate(None=Divide(A=Subtract(A=@n5.VGradient, B=@n7), B=@n6))
  n9 = Constant() {desc=Bottom Gradient Saturation, r=0.75}
  n10 = OneMinus(None=Saturate(None=Multiply(A=@n8, B=@n9) {constB=0.3}))
  Multiply(
   A=LinearInterpolate(
    A=TextureSample(UVs=@n4.Scaled To Fill, Tex=@n3).RGB
    B=Constant3Vector() {constant=<rgba 1 1 1 0>}
    Alpha=ScalarParameter() {defaultValue=1, parameterName=IsLoading}
   )
   B=Multiply(A=@n10, B=@n10)
  )
  n11 = OneMinus(None=MaterialFunctionCall() {materialFunction=&LinearGradient}.UGradient)
  n12 = Frac(None=Divide(A=Time() {bOverride_Period=true, period=2}))
  n13 = Add(A=Divide(A=Subtract(A=@n11)) {constB=1.5}, B=@n12)
  n14 = OneMinus(None=@n12)
  n15 = Multiply(A=@n14, B=@n14)
  NamedRerouteDeclaration(None=Saturate(None=Multiply(A=@n13, B=@n15))) {name=LoadingGradient, nodeColor=<rgba 0 0 0 1>}
 ]
}
